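; A scripted action that opens the Homebase news screen from a token item in the front end.
; On execute, a Token source cast to a Fort account item fires the OpenNewsScreen client event, while other sources or a failed cast lead to cancelling.
; DoOnce-gated logic, also tied to the FinishedModalQueue client event, marks that item seen through the MCP context and completes the action.
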(event ClientEvent|HandleClientEvent:FinishedModalQueue
  (params (EventSource :type Object%20Reference) (EventFocus :type Object%20Reference) (ClientEvent :type Fort%20Client%20Event%20Structure%20%28by%20ref%29))
  (Utilities|FlowControl|DoOnce true
    (:Completed
      (MCPContext|MarkItemSeen (BlueprintContext|GetFortMcpContext) (stale ScriptedAction|EventExecute _asfort_account_item))
      (ScriptedAction|CompleteAction false))))

(event ScriptedAction|EventExecute
  (params (Params :type Fort%20Scripted%20Action%20Params%20Structure%20%28by%20ref%29))
  (bind (_player _sourcetype _sourceitem) (Utilities|Struct|BreakFortScriptedActionParams Params))
  (switch Utilities|FlowControl|Switch|SwitchonEFortScriptedActionSource _sourcetype
    (:Quest
      _)
    (:Token
      (bind _asfort_account_item (Utilities|Casting|CastToFortAccountItem _sourceitem)
        (:then
          (CallFunction|ClientEvent|ExecuteClientEvent "(CategoryName=\"Homebase\",EventName=\"OpenNewsScreen\")")
          (Utilities|FlowControl|DoOnce true
            (:Completed
              (MCPContext|MarkItemSeen (BlueprintContext|GetFortMcpContext) _asfort_account_item)
              (ScriptedAction|CompleteAction false))))
        (:CastFailed
          _)))
    (:Max_None
      _))
  (ScriptedAction|CancelAction false))

(fn ConstructionScript)
